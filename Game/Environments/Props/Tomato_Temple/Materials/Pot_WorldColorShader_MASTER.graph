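# Master material for Tomato Temple pot props that varies tint by world position so placed copies differ in color.
# Summed actor position components, divided by a frequency and optionally offset by UV1, index color charts that multiply base color, optionally masked into two zones.
# Static switches add tiling or triplanar detail, UV1 decals and specular/roughness masks.
# The pot normal map is a fixed texture, and attributes feed MF_CreativeMoveTool.
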Material /Game/Environments/Props/Tomato_Temple/Materials/Pot_WorldColorShader_MASTER
graph {
 properties {
  bAutomaticallySetUsageInEditor=false
  bUseMaterialAttributes=true
  bUsedWithInstancedStaticMeshes=true
  ditheredLODTransition=true
  pixelDepthOffsetMode=PDOM_Legacy
  usedShadingModels=Default Lit
 }
 nodes [
  n1 = VectorParameter() {group=World Space Color Picker, parameterName=World Space Color Picker Color Override, defaultValue=<rgba 1 1 1 1>}
  n2 = ComponentMask(None=TextureCoordinate() {coordinateIndex=1}) {g=true}
  n3 = ScalarParameter() {
   defaultValue=0.1
   desc="1/number of colors. ie 1/10 = .1"
   group=World Space Color Picker
   parameterName="Number Of Colors in Color Chart (1/Number of colors)"
  }
  n4 = Divide(
   A=MaterialFunctionCall!(f2, f3=ActorPositionWS(), f4) {materialFunction=/Game/Packages/Fortnite_UniqueMaterials_1/VectorComponentOps}.f3 Add
   B=ScalarParameter() {defaultValue=16.71, group=World Space Color Picker, parameterName="Color Change Frequency (WS)"}
  )
  n5 = StaticSwitchParameter(True=Add(A=Multiply(A=@n2, B=@n3), B=@n4), False=@n4) {group=World Space Color Picker, parameterName=Offset Color Index with UV Channel 2 G}
  n6 = TextureCoordinate() {coordinateIndex=1}
  n7 = TextureSampleParameter2D(UVs=@n6) {group=Decal, parameterName=decal, texture=&T_BlacksRGB}
  n8 = MaterialFunctionCall(Float2=@n6) {materialFunction=&BreakOutFloat2Components}
  n9 = Constant() {r=0.5}
  n10 = Constant() {r=1}
  n11 = Multiply(A=Multiply(A=SphereMask(A=@n8.R, B=@n9, Radius=@n9, Hardness=@n10), B=SphereMask(A=@n8.G, B=@n9, Radius=@n9, Hardness=@n10)), B=@n7.A)
  n12 = TextureCoordinate()
  n13 = VectorParameter() {group=Diffuse, parameterName=Base Color Offset RG Tiling BA, defaultValue=<rgba 0 0 1 1>}
  n14 = ComponentMask(None=@n13.RGB) {g=true, r=true}
  n15 = StaticSwitchParameter(True=Multiply(A=Add(A=@n12, B=@n14), B=AppendVector(A=@n13.B, B=@n13.B)), False=@n12) {group=Diffuse, parameterName=Manipulate Base Color and Spec UVS}
  n16 = TextureSampleParameter2D(UVs=@n15) {desc=Use for exterior walls and floors, group=Diffuse, parameterName=Base Color}
  n17 = TextureObjectParameter() {group=Custom Tile, parameterName=TilingBaseColor}
  n18 = MaterialFunctionCall!(
   TextureObject=@n17
   BlendContrast=ScalarParameter() {defaultValue=6.6987, group=Custom Tile, parameterName=WorldSpaceBlendContrast}
   TextureSize=ScalarParameter() {defaultValue=572.252, group=Custom Tile, parameterName=WorldSpaceBlendSize}
  ) {materialFunction=/Game/Packages/Fortress_SharedMaterials/Base_Material_Functions/CheapTriPlanerBlend}
  n19 = TextureCoordinate()
  n20 = TextureCoordinate() {coordinateIndex=1}
  n21 = StaticBoolParameter() {group=Custom Tile, parameterName=Use2ndUVForTiling}
  n22 = TextureSample(
   UVs=Multiply(
    A=StaticSwitch(True=@n19, False=@n20, Value=@n21)
    B=ScalarParameter() {defaultValue=4, group=Custom Tile, parameterName=Base Color Pattern Tiling}
   )
   Tex=@n17
  )
  n23 = AppendVector(A=@n22.RGB, B=@n22.RGB)
  n24 = MaterialFunctionCall(Float4=StaticSwitch(
    True=QualitySwitch(Default=@n18.Result, Low=@n23, High=@n18.Result, Medium=@n18.Result)
    False=@n23
    Value=StaticBoolParameter() {group=Custom Tile, parameterName=UseTriPlaneBlend}
   )) {materialFunction=&BreakOutFloat4Components}
  n25 = AppendVector(A=AppendVector(A=@n24.R, B=@n24.R), B=@n24.B)
  n26 = Multiply(A=@n25, B=ScalarParameter() {defaultValue=0.4, group=Custom Tile, parameterName=TilingDiffuseStainDarkness})
  n27 = TextureSample(UVs=@n15) {desc=TilingTextureBlendMask, samplerType=SAMPLERTYPE_LinearColor, texture=/Game/Athena/Aircraft/Textures/T_BattleBus_CustomTextureMask}
  n28 = StaticSwitch(
   True=LinearInterpolate(A=@n16.RGB, B=LinearInterpolate(A=@n25, B=@n26, Alpha=@n27.R), Alpha=@n27.G)
   False=@n16.RGB
   Value=StaticBoolParameter() {group=Custom Tile, parameterName=UseTilingBaseColorTexture}
  )
  n29 = ComponentMask(None=StaticSwitchParameter(True=LinearInterpolate(A=@n28, B=@n7.RGB, Alpha=@n11), False=@n28) {group=Decal, parameterName=Apply decal via uv 1}) {b=true, g=true, r=true}
  n30 = Multiply(
   A=StaticSwitchParameter(
    True=@n1.RGB
    False=TextureSampleParameter2D(UVs=@n5) {group=World Space Color Picker, parameterName=Color Chart, texture=/Game/Global/Textures/Default/ColorSets/CarColors_Interiors}.RGB
   ) {group=World Space Color Picker, parameterName=Override World Space Color Picker Color}
   B=@n29
  )
  n31 = TextureSampleParameter2D(UVs=@n4) {group=World Space Color Picker, parameterName=Color Chart 2, texture=/Game/Global/Textures/Default/ColorSets/CarColors}
  n32 = TextureSampleParameter2D() {
   group=World Space Color Picker
   parameterName=World Space Color Picker Mask
   texture=/Game/Vehicles/FORT_Vehicles_01/Materials/M_FORT_Truck_01_M_New
  }
  n33 = AppendVector(A=@n32.RGB, B=@n32.RGB)
  n34 = StaticComponentMaskParameter(None=@n33) {defaultB=true, group=World Space Color Picker, parameterName=World Space Color Picker Mask Channel}
  n35 = LinearInterpolate(
   A=LinearInterpolate(A=@n30, B=Multiply(A=@n31.RGB, B=@n29), Alpha=@n34)
   B=@n29
   Alpha=StaticComponentMaskParameter(None=@n33) {defaultG=true, group=World Space Color Picker, parameterName=World Space Color Picker 2 Mask Channel}
  )
  n36 = LinearInterpolate(A=@n30, B=@n29, Alpha=@n34)
  n37 = StaticSwitchParameter(
   True=StaticSwitchParameter(
    True=StaticSwitchParameter(True=@n35, False=@n36) {group=World Space Color Picker, parameterName=Use 2 World Position Coloring Colors}
    False=@n30
   ) {group=World Space Color Picker, parameterName=Mask World Position Color Picker}
   False=@n29
  ) {group=World Space Color Picker, parameterName=Use World Position Coloring}
  n38 = Multiply(A=@n37, B=VectorParameter() {group=Diffuse, parameterName=Base Color Multiplier, defaultValue=<rgba 1 1 1 1>}.RGB)
  n39 = TextureSampleParameter2D() {
   group=World Space Color Picker
   parameterName=World Space Color Picker Mask
   texture=&M_FORT_Truck_01_M_New
  }
  n40 = StaticSwitchParameter(True=LinearInterpolate(A=@n38, B=@n37, Alpha=@n39.G), False=@n38) {group=Diffuse, parameterName=Modulate Base With Mask}
  n41 = TextureSampleParameter2D(UVs=@n15) {
   group=Specular
   parameterName=SpecularMasks
   samplerType=SAMPLERTYPE_Masks
   texture=&T_Fortnite_Default_S
  }
  n42 = StaticSwitchParameter(True=ComponentMask(None=@n41.RGB) {b=true, g=true, r=true}, False=Constant3Vector() {constant=<rgba 1 0 0.2667 0>}) {desc=Specular Power, group=Specular, parameterName=Use SpecularMasks}
  n43 = SetMaterialAttributes!(
   MaterialAttributes
   Base Color=MaterialFunctionCall!(S, V2, V3=StaticSwitchParameter(True=@n40, False=@n37) {group=Diffuse, parameterName=Modulate Base Color}, V4) {desc=Diffuse, materialFunction=/Engine/Functions/Engine_MaterialFunctions02/PassThrough}.V3
   Specular=ComponentMask(None=@n42) {desc=spec, r=true}
   Roughness=ComponentMask(None=@n42) {b=true, desc=roughness}
   Normal=TextureSample() {samplerType=SAMPLERTYPE_Normal, texture=/Game/Environments/Props/Tomato_Temple/Textures/T_Tomato_Pots_N}.RGB
  ) {attributeSetTypes [
    #1
    #2
    #3
    #4
   ]}
  MaterialFunctionCall(MA In=@n43) {materialFunction=&MF_CreativeMoveTool}
  VertexInterpolator(VS=MaterialFunctionCall() {materialFunction=&ComponentPivotLocation}.Result)
 ]
}
